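# Sound cue for the dematerialize sound played when an NPC is eliminated.
# A random node picks one of four equally weighted wave variations, fed through a short fixed delay and a modulator that slightly randomizes pitch upward.
# It uses a dedicated deres attenuation asset and the gadgets sound class, with a concurrency override limiting simultaneous instances.
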SoundCue /Game/Athena/Sounds/NPCs/NPC_DeRes_Death_Cue
root=Modulator_0
nodes [
 {id=Delay_0, children [Random_0], properties {delayMax=0.1, delayMin=0.1}}
 {id=Modulator_0, children [Delay_0], properties {pitchMax=1.1, pitchMin=1}}
 {id=Random_0, children [WavePlayer_1,WavePlayer_2,WavePlayer_3,WavePlayer_4], properties {weights [1,1,1,1]}}
 {id=WavePlayer_1, properties {soundWaveAssetPtr=&/NPC_Eliminated_DeMaterialize_01}}
 {id=WavePlayer_2, properties {soundWaveAssetPtr=&/NPC_Eliminated_DeMaterialize_02}}
 {id=WavePlayer_3, properties {soundWaveAssetPtr=&/NPC_Eliminated_DeMaterialize_03}}
 {id=WavePlayer_4, properties {soundWaveAssetPtr=&/NPC_Eliminated_DeMaterialize_04}}
]
properties {
 attenuationSettings=&/NPC_Deres_Att
 bOverrideConcurrency=true
 maxDistance=2750
 soundClassObject=&Fort_Gadgets
 volumeMultiplier=1
 concurrencyOverrides {maxCount=2}
}
