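# Sound cue for the de-res death effect of wildlife NPCs, playing the glitch death sound.
# A single wave player feeds a modulator that adds slight upward pitch variation.
# The cue uses glitch NPC attenuation at half volume, routes to the gadgets sound class, and overrides concurrency to cap simultaneous instances at a low count.
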SoundCue /Game/Athena/Sounds/NPCs/NPC_DeRes_Wildlife_Death_Cue
root=Modulator_0
nodes [
 {id=Modulator_0, children [WavePlayer_0], properties {pitchMax=1.03, pitchMin=1}}
 {id=WavePlayer_0, properties {soundWaveAssetPtr=&/NPC_Glitch_VFX_Death}}
]
properties {
 attenuationSettings=&/NPC_Glitch_Att
 bOverrideConcurrency=true
 maxDistance=1500
 soundClassObject=&Fort_Gadgets
 volumeMultiplier=0.5
 concurrencyOverrides {maxCount=2}
}
