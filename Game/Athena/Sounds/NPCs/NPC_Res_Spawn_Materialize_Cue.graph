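# Sound cue for an NPC materializing on spawn, using "RezIn" sound waves.
# A Random node picks one of four equally weighted wave variations and feeds a Modulator that adds a slight upward pitch variation.
# It plays in the Fort_Gadgets sound class with NPC glitch attenuation, and overrides concurrency to limit simultaneous instances to a small count.
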SoundCue /Game/Athena/Sounds/NPCs/NPC_Res_Spawn_Materialize_Cue
root=Modulator_0
nodes [
 {id=Modulator_0, children [Random_0], properties {pitchMax=1.03, pitchMin=1}}
 {id=Random_0, children [WavePlayer_0,WavePlayer_3,WavePlayer_4,WavePlayer_5], properties {weights [1,1,1,1]}}
 {id=WavePlayer_0, properties {soundWaveAssetPtr=&/NPC_Spawn_Materialize_RezIn_04}}
 {id=WavePlayer_3, properties {soundWaveAssetPtr=&/NPC_Spawn_Materialize_RezIn_03}}
 {id=WavePlayer_4, properties {soundWaveAssetPtr=&/NPC_Spawn_Materialize_RezIn_02}}
 {id=WavePlayer_5, properties {soundWaveAssetPtr=&/NPC_Spawn_Materialize_RezIn_01}}
]
properties {
 attenuationSettings=&/NPC_Glitch_Att
 bOverrideConcurrency=true
 maxDistance=1500
 soundClassObject=&Fort_Gadgets
 volumeMultiplier=1
 concurrencyOverrides {maxCount=2}
}
